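# Unlit translucent Niagara sprite and mesh material for the Serpent Coil pickaxe's idle effect.
# A panning, tiled Horned Judgment spike texture, sampled through UVs pinched at both V ends, blends two colors via a curve atlas, tinted by particle color.
# Opacity combines texture alpha, particle alpha, a one-ended V falloff and view-angle fade, plus depth fade off mobile.
# A time-driven sine wave, masked by UVs, provides vertex offset along a vector parameter.
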Material /BRCosmetics/Weapons/FORT_Melee/Pickaxe_SerpentCoil/FX/M_Pickaxe_SerpentCoil_Idle
graph {
 properties {
  bUsedWithNiagaraMeshParticles=true
  bUsedWithNiagaraSprites=true
  blendMode=BLEND_Translucent
  pixelDepthOffsetMode=PDOM_Legacy
  shadingModel=MSM_Unlit
  usedShadingModels=Unlit
 }
 nodes [
  n1 = VectorParameter() {group=Color, parameterName=Color 1, defaultValue=<rgba 0.75 0.75 0.75 1>}
  n2 = VectorParameter() {group=Color, parameterName=Color 2, defaultValue=<rgba 3 3 3 1>}
  n3 = MaterialFunctionCall(XY=TextureCoordinate()) {materialFunction=&Swizzle}
  n4 = ComponentMask(None=TextureCoordinate()) {g=true}
  n5 = ScalarParameter() {defaultValue=0.45, group=Mask, parameterName=Start Shirnk, sliderMax=1, sortPriority=0}
  n6 = Power(Base=OneMinus(None=@n4), Exp=ScalarParameter() {defaultValue=1, group=Mask, parameterName=End Shirnk, sliderMax=1, sortPriority=0}) {constExponent=0.3}
  n7 = AppendVector(A=Constant() {r=1}, B=Multiply(A=Power(Base=@n4, Exp=@n5) {constExponent=0.3}, B=@n6) {constA=1})
  n8 = Saturate(None=Add(A=Divide(A=@n3.YX, B=@n7), B=Constant2Vector()))
  n9 = AppendVector(
   A=ScalarParameter() {defaultValue=0.7, group=Mask, parameterName=Speed X, sliderMax=3}
   B=ScalarParameter() {group=Mask, parameterName=Speed Y, sliderMax=3}
  )
  n10 = ScalarParameter() {defaultValue=1, group=Mask, parameterName=Tile X, sliderMax=10}
  n11 = ScalarParameter() {defaultValue=16, group=Mask, parameterName=Tile Y, sliderMax=10}
  n12 = TextureSampleParameter2D!(UVs=Multiply(A=Panner(Coordinate=@n8, Speed=@n9) {speedX=0.5}, B=AppendVector(A=@n10, B=@n11)), Level, Apply View MipBias) {
   group=Mask
   mipValueMode=TMVM_MipLevel
   parameterName=Texture
   texture=/BRCosmetics/Weapons/FORT_Melee/Pickaxe_HornedJudgment/FX/T_Pickaxe_HornedJudgement_Spikes03
  }
  n13 = CurveAtlasRowParameter(CurveTime=Power(Base=@n12.R, Exp=ScalarParameter() {defaultValue=0.5, group=Color, parameterName=Color Power, sliderMax=5})) {
   atlas=/Game/Weapons/FORT_Melee/Pickaxe_HornedJudgment/FX/CurveAtlas_HornedJudgement
   curve=/Game/Weapons/FORT_Melee/Pickaxe_HornedJudgment/FX/Curve_Pickaxe_HornedJudgement_Spike
   group=Color
   parameterName=Curve_Color
  }
  n14 = ParticleColor()
  MaterialFunctionCall(Emissive=Multiply(A=LinearInterpolate(A=@n1.RGB, B=@n2.RGB, Alpha=@n13.R), B=@n14.RGB)) {materialFunction=&MF_FX_EmissiveAdjustments}
  n15 = TextureCoordinate()
  n16 = ComponentMask(None=@n15) {r=true}
  n17 = ComponentMask(None=@n15) {g=true}
  n18 = Sine(None=@n17) {period=2}
  n19 = Add(A=@n17, B=Time())
  n20 = ScalarParameter() {defaultValue=2, parameterName=Period, sliderMax=10}
  n21 = ConstantBiasScale(None=Sine(None=Multiply(A=@n19, B=@n20)))
  n22 = ScalarParameter() {defaultValue=30, parameterName=WPO Intensity, sliderMax=50}
  Multiply(
   A=Multiply(A=Multiply(A=@n16, B=@n18), B=Multiply(A=@n21, B=@n22))
   B=Transform(None=VectorParameter() {parameterName=WPO Tangent Vector, defaultValue=<rgba 1 1 1 1>}.RGB)
  )
  n23 = Multiply(A=@n6) {constB=6}
  n24 = Multiply(
   A=Multiply(A=@n14.A, B=Multiply(A=@n12.A, B=@n23))
   B=MaterialFunctionCall("FadeSensitivity (optional)"=ScalarParameter() {defaultValue=0.75, parameterName=Fade Sensitivity, sliderMax=20}) {materialFunction=&FadeBasedOnViewAngle}.Result
  )
  ShadingPathSwitch(Default=DepthFade(Opacity=@n24, FadeDistance=ScalarParameter() {defaultValue=2, parameterName=FadeDistance}), Mobile=@n24) {desc=Migrated from FeatureLevelSwitch}
 ]
}
