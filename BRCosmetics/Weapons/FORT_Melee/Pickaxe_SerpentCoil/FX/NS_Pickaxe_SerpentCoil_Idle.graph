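# Niagara idle effect for the Serpent Coil pickaxe, adding glowing spike meshes and an air glow that follow the weapon's skeletal mesh.
# Panning_Spikes and Tip_Spikes burst mesh particles, while Airglow bursts sprites spread along a curved arc by execution-index curves.
# Particles outlive their lifetime and re-sample the skeletal mesh each update.
# System-level variables supply colors, and time-of-day scalars adjust spike brightness and glow alpha.
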NiagaraSystem /BRCosmetics/Weapons/FORT_Melee/Pickaxe_SerpentCoil/FX/NS_Pickaxe_SerpentCoil_Idle
dependencies {
 usedDataInterfaces [&NiagaraDataInterfaceCurve, &NiagaraDataInterfaceSkeletalMesh, &NiagaraDataInterfaceArrayColor]
 usedDynamicInputs [&SelectLinearColorFromArray, &MakeIntFromFloat]
 usedModules [
  &self:SystemSpawnScript.SystemScriptSource.SystemScriptGraph.NiagaraNodeAssignment_0.SetVariables_#1
  &self:SystemSpawnScript.SystemScriptSource.SystemScriptGraph.NiagaraNodeAssignment_1.SetVariables_#2
  &SystemState
 ]
}
emitters {
 Airglow {
  emitterUpdateScript [
   &EmitterState()
   &SpawnBurst_Instantaneous(Spawn Count=12, Use Spawn Probability=false, UseLoopCountLimit=true)
  ]
  particleSpawnScript [
   &InitializeParticle(
    Color=$System.AirGlow Color
    Lifetime=0.2
    Position Offset=<0 204 16.333>
    Sprite Size Mode=::NewEnumerator3
    UsePositionOffset=false
    Write Material Random=true
   )
   &CameraOffset(Camera Offset Amount=-5)
  ]
  particleUpdateScript [
   &ParticleState(Kill Particles When Lifetime Has Elapsed=false)
   &SkeletalMeshLocation(
    Bone Sampling Mode=::NewEnumerator4
    MaskSpawnGroup=false
    Orientation Sampling=::NewEnumerator1
    Sampled Position Offset=&MakeVector(
     X=0
     Y=&FloatFromCurve(
      FloatCurve=DataInterface() {propertyValues {Curve {keys [
          {
           arriveTangent=-21.7751
           arriveTangentWeight=4.4149
           interpMode=RCIM_Cubic
           leaveTangent=-21.7751
           leaveTangentWeight=4.4149
           tangentMode=RCTM_User
           tangentWeightMode=RCTWM_WeightedBoth
           time=0
           value=0
          }
          {
           arriveTangent=23.3787
           arriveTangentWeight=11.9999
           interpMode=RCIM_Cubic
           leaveTangent=23.3787
           leaveTangentWeight=11.9999
           tangentMode=RCTM_User
           tangentWeightMode=RCTWM_WeightedBoth
           time=1
           value=-19
          }
         ]}}}
      CurveIndex=&ReturnNormalizedExecIndex()
     )
     Z=&FloatFromCurve(
      FloatCurve=DataInterface() {propertyValues {Curve {keys [
          {
           arriveTangent=10.2629
           interpMode=RCIM_Cubic
           leaveTangentWeight=0.3333
           tangentMode=RCTM_User
           tangentWeightMode=RCTWM_WeightedBoth
           time=0
           value=-14
          }
          {
           arriveTangent=41.0832
           arriveTangentWeight=20.4505
           interpMode=RCIM_Cubic
           leaveTangent=41.0832
           leaveTangentWeight=20.4505
           tangentMode=RCTM_User
           tangentWeightMode=RCTWM_WeightedBoth
           time=1
           value=17
          }
         ]}}}
      CurveIndex=&ReturnNormalizedExecIndex()
     )
    )
    Skeletal Mesh=$System.Skeletal Mesh
   )
   &ScaleSpriteSize(
    Uniform Curve Index=&ReturnNormalizedExecIndex()
    Uniform Curve Sprite Scale=DataInterface() {propertyValues {Curve {keys [
        {arriveTangent=-10.0586, interpMode=RCIM_Cubic, leaveTangent=-10.0586, tangentMode=RCTM_User, time=0, value=2}
        {interpMode=RCIM_Cubic, tangentMode=RCTM_User, time=0.2, value=1.25}
        {arriveTangent=2.1343, interpMode=RCIM_Cubic, leaveTangent=2.1343, tangentMode=RCTM_User, time=0.5, value=1.5}
        {arriveTangent=-5.3184, interpMode=RCIM_Cubic, leaveTangent=-5.3184, tangentMode=RCTM_User, time=1, value=1.2}
       ]}}}
   )
   &DynamicMaterialParameters(
    Index 0 Param 1=2
    Module.Index0Param0Write=true
    Module.Index0Param1Write=true
    Module.Index0Param2Write=true
    Module.Index0Param3Write=true
   )
   &ScaleColor(
    Scale Alpha=&FNTimeOfDayScalar(Night Scale=0.05, Morning Scale=0.05, Day Scale=0.1, Evening Scale=0.1)
    ScaleA=true
    ScaleRGB=true
   )
  ]
  rendererClasses [&NiagaraSpriteRendererProperties]
  renderers [{rendererClass=&NiagaraSpriteRendererProperties}]
 }
 Panning_Spikes {
  emitterUpdateScript [
   &EmitterState()
   &SpawnBurst_Instantaneous(Use Spawn Probability=false, UseLoopCountLimit=true)
  ]
  particleSpawnScript [
   &InitializeParticle(
    Color=$System.Color One
    Lifetime=0.2
    Mesh Scale Mode=::NewEnumerator3
    Position Offset=<0 204 16.333>
    UsePositionOffset=false
    Write Material Random=true
   )
   &CameraOffset()
  ]
  particleUpdateScript [
   &ParticleState(Kill Particles When Lifetime Has Elapsed=false)
   &SkeletalMeshLocation(Bone Sampling Mode=::NewEnumerator4, MaskSpawnGroup=false, Orientation Sampling=::NewEnumerator1, Skeletal Mesh=$System.Skeletal Mesh)
   &ScaleColor(
    Scale RGB=&VectorFromFloat(Value=&FNTimeOfDayScalar(Night Scale=0.85, Morning Scale=1, Evening Scale=1))
    ScaleA=true
    ScaleRGB=true
   )
  ]
  rendererClasses [&NiagaraMeshRendererProperties]
  renderers [{rendererClass=&NiagaraMeshRendererProperties, rendererIndex=0}]
 }
 Tip_Spikes {
  emitterUpdateScript [
   &EmitterState()
   &SpawnBurst_Instantaneous(Use Spawn Probability=false, UseLoopCountLimit=true)
  ]
  particleSpawnScript [
   &InitializeParticle(
    Color=$System.Color One
    Lifetime=0.2
    Mesh Scale Mode=::NewEnumerator3
    Mesh Uniform Scale=1.025
    Position Offset=<0 204 16.333>
    UsePositionOffset=false
    Write Material Random=true
   )
   &CameraOffset(Camera Offset Amount=-1)
  ]
  particleUpdateScript [
   &ParticleState(Kill Particles When Lifetime Has Elapsed=false)
   &SkeletalMeshLocation(Bone Sampling Mode=::NewEnumerator4, MaskSpawnGroup=false, Orientation Sampling=::NewEnumerator1, Skeletal Mesh=$System.Skeletal Mesh)
   &ScaleColor(
    Scale RGB=&VectorFromFloat(Value=&FNTimeOfDayScalar(Night Scale=0.75, Morning Scale=1, Day Scale=2, Evening Scale=1))
    ScaleA=true
    ScaleRGB=true
   )
  ]
  rendererClasses [&NiagaraMeshRendererProperties]
  renderers [{rendererClass=&NiagaraMeshRendererProperties, rendererIndex=0}]
 }
}
summary {
 systemName=NS_Pickaxe_SerpentCoil_Idle
 emitters [
  {emitterName=Panning_Spikes, rendererClasses [&NiagaraMeshRendererProperties]}
  {emitterName=Tip_Spikes, rendererClasses [&NiagaraMeshRendererProperties]}
  {emitterName=Airglow, rendererClasses [&NiagaraSpriteRendererProperties]}
 ]
 userVariables [
  User.bIsFrontEnd {type=NiagaraBool}
  User.bIsFrontEndPreview {type=NiagaraBool}
  User.bisSwinging {type=NiagaraBool}
  User.Spawn Delay {type=NiagaraFloat, defaultValue=0, description=""}
  User.Style {type=NiagaraFloat, defaultValue=0, description=""}
 ]
}
