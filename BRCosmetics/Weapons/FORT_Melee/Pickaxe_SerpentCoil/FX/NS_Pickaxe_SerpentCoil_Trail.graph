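# Niagara swing trail for the Serpent Coil pickaxe, built from two ribbon emitters that use the shared AnimTrail modules on skeletal mesh bones or sockets.
# Flame_Trail is a short ribbon with a bright initial flash, a randomized material parameter and opacity faded by system age.
# Flame_Trail_Dark is wider, has four ribbon renderers, and uses User.Style to set main, Galaxy and Holo flags, choose edge tinting and swap bone order for Galaxy.
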NiagaraSystem /BRCosmetics/Weapons/FORT_Melee/Pickaxe_SerpentCoil/FX/NS_Pickaxe_SerpentCoil_Trail
dependencies {
 usedDataInterfaces [&NiagaraDataInterfaceCurve, &NiagaraDataInterfaceColorCurve, &NiagaraDataInterfaceArrayColor]
 usedDynamicInputs [&SelectLinearColorFromArray, &MakeIntFromFloat]
 usedModules [
  &self:SystemSpawnScript.SystemScriptSource.SystemScriptGraph.NiagaraNodeAssignment_4.SetVariables_#1
  &SystemState
 ]
}
emitters {
 Flame_Trail {
  emitterUpdateScript [
   &EmitterState()
   &FN_SpawnAnimTrail_v2()
   /SetVariables(Emitter.RandomFloat:NiagaraFloat=&RandomRangeFloat(Minimum=0, Maximum=1)) {isSetParameters=true}
  ]
  particleSpawnScript [
   /Game/Effects/Niagara/Modules/AnimTrail/FN_InitializeAnimTrail/FN_InitializeAnimTrailBoneIndecies(Second Bone or Socket Index=1)
   /Game/Effects/Niagara/Modules/AnimTrail/FN_InitializeAnimTrail/FN_InitializeAnimTrail(AnimTrailWidth=0.9, Lifetime=0.27)
   /SetVariables(Particles.Opacity:NiagaraFloat=&FloatFromCurve(
     FloatCurve=DataInterface() {propertyValues {Curve {keys [
         {arriveTangent=-2.2886, interpMode=RCIM_Cubic, leaveTangent=-2.2886, tangentMode=RCTM_User, time=0.15, value=1}
         {arriveTangent=-3.373, interpMode=RCIM_Cubic, leaveTangent=-3.373, tangentMode=RCTM_User, time=0.3, value=0}
        ]}}}
     CurveIndex=$Engine.System.Age
     Scale Curve=2
    )) {isSetParameters=true}
  ]
  particleUpdateScript [
   &FN_UpdateAnimTrail_v2()
   &ScaleColor(
    Scale Alpha=&FloatFromCurve(FloatCurve=DataInterface() {propertyValues {Curve {keys [
         {arriveTangent=2.5517, interpMode=RCIM_Cubic, time=0, value=1}
         {arriveTangent=-4.7804, interpMode=RCIM_Cubic, leaveTangent=-4.7804, tangentMode=RCTM_User, time=0.7, value=0}
        ]}}})
    Scale RGB=&MakeVectorFromLinearColorRGB(Color=&ColorFromCurve(ColorCurve=DataInterface() {propertyValues {
        AlphaCurve {keys [<at 0 1 RCIM_Cubic>]}
        BlueCurve {keys [{arriveTangent=-21.6049, interpMode=RCIM_Cubic, time=0, value=3}, <at 0.1478 0.35>]}
        GreenCurve {keys [{arriveTangent=-21.6049, interpMode=RCIM_Cubic, time=0, value=3}, <at 0.1478 0.35>]}
        RedCurve {keys [{arriveTangent=-21.6049, interpMode=RCIM_Cubic, time=0, value=3}, <at 0.1478 0.35>]}
       }}))
    ScaleA=true
    ScaleRGB=true
   )
   &ScaleRibbonWidth(Ribbon Width Scale=&FloatFromCurve(FloatCurve=DataInterface() {propertyValues {Curve {keys [<at 0.5 1 RCIM_Cubic>, {arriveTangent=-0.7564, interpMode=RCIM_Cubic, leaveTangent=-0.7564, tangentMode=RCTM_User, time=1, value=0}]}}}))
   &SolveForcesAndVelocity(Clamp Velocity=false, Limit Acceleration=false)
   &DynamicMaterialParameters(
    Index 0 Param 1=$Emitter.RandomFloat
    Index 0 Param 4=&FloatFromCurve(
     FloatCurve=DataInterface() {propertyValues {Curve {keys [
         {arriveTangent=2.013, interpMode=RCIM_Cubic, leaveTangent=2.013, tangentMode=RCTM_User, time=0, value=0}
         {arriveTangent=21.3231, interpMode=RCIM_Cubic, leaveTangent=-0.6194, tangentMode=RCTM_Break, time=0.15, value=1}
         {arriveTangent=-0.0405, interpMode=RCIM_Cubic, leaveTangent=-0.0405, tangentMode=RCTM_User, time=1, value=0}
        ]}}}
     Scale Curve=$Particles.Opacity
    )
    Module.Index0Param0Write=true
    Module.Index0Param1Write=false
    Module.Index0Param2Write=true
    Module.Index0Param3Write=true
   )
  ]
  rendererClasses [&NiagaraRibbonRendererProperties]
  renderers [{rendererClass=&NiagaraRibbonRendererProperties}]
 }
 Flame_Trail_Dark {
  emitterUpdateScript [
   &EmitterState()
   &FN_SpawnAnimTrail_v2()
   /SetVariables(Emitter.RandomFloat:NiagaraFloat=&RandomRangeFloat(Minimum=0, Maximum=1)) {isSetParameters=true}
   /SetVariables(
    Emitter.Main Bool:NiagaraBool=&Boolean_AndOperation(
     Bool1=&SetBoolByFloatComparison(A=$User.Style, B=3, Comparison Type=::NewEnumerator5)
     Bool2=&SetBoolByFloatComparison(A=$User.Style, B=5, Comparison Type=::NewEnumerator5)
    )
    Emitter.Variant Galaxy Bool:NiagaraBool=&SetBoolByFloatComparison(A=$User.Style, B=5, Comparison Type=::NewEnumerator4)
    Emitter.Variant Holo Bool:NiagaraBool=&SetBoolByFloatComparison(A=$User.Style, B=3, Comparison Type=::NewEnumerator4)
   ) {isSetParameters=true}
  ]
  particleSpawnScript [
   &FN_InitializeAnimTrailBoneIndecies(
    First Bone or Socket Index:NiagaraInt32=&MakeCustomIntFromBool(Bool=&SetBoolByFloatComparison(A=$User.Style, B=5, Comparison Type=::NewEnumerator4), True INT=0, False INT=1)
    Second Bone or Socket Index:NiagaraInt32=&MakeCustomIntFromBool(Bool=&SetBoolByFloatComparison(A=$User.Style, B=5, Comparison Type=::NewEnumerator4), True INT=1, False INT=0)
   )
   &FN_InitializeAnimTrail(
    AnimTrailWidth=1.15
    Color=&LerpLinearColors(
     StartColor=<rgba 1 1 1 1>
     EndColor=$System.Edge Color
     LerpFactor=&MakeCustomFloatFromBool(
      Bool=&Boolean_AndOperation(
       Bool1=&SetBoolByFloatComparison(A=$User.Style, B=3, Comparison Type=::NewEnumerator5)
       Bool2=&SetBoolByFloatComparison(A=$User.Style, B=5, Comparison Type=::NewEnumerator5)
      )
      True Float=1
      False Float=0
     )
    )
    Lifetime=0.23
   )
  ]
  particleUpdateScript [
   &FN_UpdateAnimTrail_v2()
   &ScaleColor(
    Scale Alpha=&FloatFromCurve(FloatCurve=DataInterface() {propertyValues {Curve {keys [
         {arriveTangent=-2.2426, interpMode=RCIM_Cubic, leaveTangent=-2.2426, tangentMode=RCTM_User, time=0.15, value=1}
         {interpMode=RCIM_Cubic, leaveTangent=-3.7744, time=0.8, value=0}
        ]}}})
    Scale RGB=&VectorFromFloat(Value=&FloatFromCurve(FloatCurve=DataInterface() {propertyValues {Curve {keys [{tangentMode=RCTM_User, time=0, value=1}, {tangentMode=RCTM_User, time=1, value=0.5}]}}}))
    ScaleA=true
    ScaleRGB=true
   )
   &ScaleRibbonWidth(Ribbon Width Scale=&FloatFromCurve(FloatCurve=DataInterface() {propertyValues {Curve {keys [{interpMode=RCIM_Cubic, tangentMode=RCTM_User, time=0, value=1}, {interpMode=RCIM_Cubic, tangentMode=RCTM_User, time=1, value=0}]}}}))
   &SolveForcesAndVelocity(Clamp Velocity=false, Limit Acceleration=false)
   &DynamicMaterialParameters(
    Index 3 Param 4=&FloatFromCurve(FloatCurve=DataInterface() {propertyValues {Curve {keys [
         {tangentMode=RCTM_User, time=0, value=1}
         {arriveTangent=-0.0405, interpMode=RCIM_Cubic, leaveTangent=-0.0405, tangentMode=RCTM_User, time=0.5, value=1}
         {arriveTangent=-0.2132, interpMode=RCIM_Cubic, leaveTangent=-0.2132, tangentMode=RCTM_User, time=1, value=0}
        ]}}})
    Module.Index3Param0Write=true
    Module.Index3Param1Write=true
    Module.Index3Param2Write=true
    Module.Index3Param3Write=true
    Write Parameter Index 0=false
    Write Parameter Index 3=true
   )
  ]
  rendererClasses [&NiagaraRibbonRendererProperties]
  renderers [
   {rendererClass=&NiagaraRibbonRendererProperties}
   {rendererClass=&NiagaraRibbonRendererProperties, rendererIndex=1}
   {rendererClass=&NiagaraRibbonRendererProperties, rendererIndex=2}
   {rendererClass=&NiagaraRibbonRendererProperties, rendererIndex=3}
  ]
 }
}
summary {
 systemName=NS_Pickaxe_SerpentCoil_Trail
 emitters [
  {emitterName=Flame_Trail, rendererClasses [&NiagaraRibbonRendererProperties]}
  {emitterName=Flame_Trail_Dark, rendererClasses [&NiagaraRibbonRendererProperties]}
 ]
 userVariables [
  User.bIsFrontEnd {type=NiagaraBool}
  User.bIsFrontEndPreview {type=NiagaraBool}
  User.RibbonIDOffset {type=NiagaraInt32}
  User.SkeletalMesh {
   type=NiagaraDataInterfaceSkeletalMesh
   defaultValue {
    struct=NiagaraExt_VariableValue_DataInterface
    value {dataInterface=&self:NiagaraDataInterfaceSkeletalMesh_1}
   }
  }
  User.Style {type=NiagaraFloat, defaultValue=0, description=""}
  User.Variant {type=NiagaraFloat, defaultValue=0, description=""}
 ]
}
